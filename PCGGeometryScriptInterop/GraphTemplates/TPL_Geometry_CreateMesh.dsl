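; PCG template graph that builds an example mesh in editor, then saves it as a static mesh or spawns it as a dynamic mesh, per SaveAsset.
; Editor-only Geometry Script nodes append spheres at cube vertices and merge them with a boolean subtraction of transformed cubes.
; At runtime or when saving, a path string from AssetPath and AssetName feeds a Static Mesh Spawner, ordered after saving by execution dependency.
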(graph TPL_Geometry_CreateMesh
  (params (SaveAsset :type Bool) (AssetName :type String :default "Asset") (AssetPath :type String :default "/Game"))
  ;[0 0 0 0.5] Spawn Static Mesh at Runtime (asset must be saved once) OR if the Save Asset is enabled in Editor
  (bind GetExecutionContext_1 (Get%20Execution%20Context%20Info))
  (bind AttributeBooleanOp_0 (Attribute%20Boolean%20Op :Operation Or :InB SaveAsset))
  (bind Branch_3 (Branch))
  (bind CreateConstant_0 (Create%20Constant :AttributeTypes (struct :Type String :StringValue "{Path}/{Name}.{Name}")))
  (bind CreateConstant_1 (Create%20Constant :AttributeTypes (struct :Type String :StringValue "{Path}")))
  (bind AttributeStringOp_0 (Attribute%20String%20Op :Operation Replace :Replace AssetPath))
  (bind CreateConstant_2 (Create%20Constant :AttributeTypes (struct :Type String :StringValue "{Name}")))
  (bind AttributeStringOp_1 (Attribute%20String%20Op :Operation Replace :Replace AssetName))
  (bind CreatePoints_1 (Create%20Points :CoordinateSpace LocalComponent))
  (bind AddAttribute_0 (Add%20Attribute :OutputTarget Mesh))
  (bind TransformPoints_3 (Transform%20Points))
  (bind StaticMeshSpawner_0 (Static%20Mesh%20Spawner :MeshSelectorParameters (class:/Script/PCG.PCGMeshSelectorByAttribute :AttributeName Mesh) :MeshSelectorType /Script/PCG.PCGMeshSelectorByAttribute))
  ;[0 0 0 0.5] Execute Dynamic Mesh ops in Editor only
  (bind GetExecutionContext_0 (Get%20Execution%20Context%20Info :Mode IsEditor))
  (bind Branch_1 (Branch))
  ;[0 0 0 0.5] Create a mesh from dynamic mesh operations (Editor Only)
  (bind StaticMeshToDynamicMeshElement_1 (Mesh%20To%20Dynamic%20Mesh%20Element :Mesh /Engine/BasicShapes/Sphere.Sphere))
  (bind DynamicMeshTransform_1 (Dynamic%20Mesh%20Transform :Transform (struct :Scale3D (0.2 0.2 0.2))))
  (bind StaticMeshToDynamicMeshElement_0 (Mesh%20To%20Dynamic%20Mesh%20Element :Mesh /PCG/DebugObjects/PCG_Cube.PCG_Cube))
  (bind MeshSampler_1 (Mesh%20Sampler :bExtractMeshFromInput true :bHasDefaultColorOverride false :SamplingMethod OnePointPerVertex))
  (bind TransformPoints_1 (Transform%20Points :ScaleMax (3.0 3.0 3.0)))
  (bind AppendMeshesFromPoints_0 (Append%20Meshes%20From%20Points :Mode DynamicMesh))
  (bind DynamicMeshTransform_0 (Dynamic%20Mesh%20Transform :Transform (struct :Rotation (0.011376107230961963 -0.08641011328638253 0.1300295006517181 0.9876721143508953) :Translation (0.0 0.0 1e+02) :Scale3D (0.6 0.4 0.6))))
  (bind DynamicMeshTransform_2 (Dynamic%20Mesh%20Transform :Transform (struct :Rotation (0.01137610723096192 -0.08641011328638237 0.13002950065171778 0.9876721143508955) :Translation (0.0 0.0 1e+02) :Scale3D (0.4 1.0 0.4))))
  (bind BooleanOperation_0 (Boolean%20Operation :BooleanOperation Subtract))
  (bind MergeDynamicMeshes_0 (Merge%20Dynamic%20Meshes))
  ;[0 0 0 0.5] Save to a Static Mesh asset or spawn Dynamic Mesh (Editor Only)
  (bind Branch_0 (Branch :bOutputToB-param SaveAsset))
  (bind SaveDynamicMeshToAsset_1 (Save%20Dynamic%20Mesh%20To%20Asset :AssetName AssetName :AssetPath AssetPath))
  (bind SpawnDynamicMesh_0 (Spawn%20Dynamic%20Mesh))
  (connect AttributeStringOp_1 Out AddAttribute_0 Attributes)
  (connect CreatePoints_1 Out AddAttribute_0 In)
  (connect DynamicMeshTransform_1 Out AppendMeshesFromPoints_0 AppendDynMesh)
  (connect StaticMeshToDynamicMeshElement_0 Out AppendMeshesFromPoints_0 InDynMesh)
  (connect TransformPoints_1 Out AppendMeshesFromPoints_0 InPoints)
  (connect GetExecutionContext_1 Out AttributeBooleanOp_0 InA)
  (connect CreateConstant_0 Out AttributeStringOp_0 In)
  (connect CreateConstant_1 Out AttributeStringOp_0 Search)
  (connect AttributeStringOp_0 Out AttributeStringOp_1 In)
  (connect CreateConstant_2 Out AttributeStringOp_1 Search)
  (connect DynamicMeshTransform_0 Out BooleanOperation_0 InA)
  (connect DynamicMeshTransform_2 Out BooleanOperation_0 InB)
  (connect MergeDynamicMeshes_0 Out Branch_0 In)
  (connect GetExecutionContext_0 Out Branch_1 bOutputToB)
  (connect GetExecutionContext_0 Out Branch_1 In)
  (connect AttributeBooleanOp_0 Out Branch_3 bOutputToB)
  (connect AttributeBooleanOp_0 Out Branch_3 In)
  (connect Branch_3 Output%20B CreateConstant_0 Execution%20Dependency)
  (connect Branch_3 Output%20B CreateConstant_1 Execution%20Dependency)
  (connect Branch_3 Output%20B CreateConstant_2 Execution%20Dependency)
  (connect Branch_3 Output%20B CreatePoints_1 Execution%20Dependency)
  (connect SaveDynamicMeshToAsset_1 OutAssetPath CreatePoints_1 Execution%20Dependency)
  (connect StaticMeshToDynamicMeshElement_0 Out DynamicMeshTransform_0 In)
  (connect StaticMeshToDynamicMeshElement_1 Out DynamicMeshTransform_1 In)
  (connect StaticMeshToDynamicMeshElement_0 Out DynamicMeshTransform_2 In)
  (connect AppendMeshesFromPoints_0 Out MergeDynamicMeshes_0 In)
  (connect BooleanOperation_0 Out MergeDynamicMeshes_0 In)
  (connect StaticMeshToDynamicMeshElement_0 Out MeshSampler_1 In)
  (connect Branch_0 Output%20B SaveDynamicMeshToAsset_1 In)
  (connect Branch_0 Output%20A SpawnDynamicMesh_0 In)
  (connect TransformPoints_3 Out StaticMeshSpawner_0 In)
  (connect Branch_1 Output%20B StaticMeshToDynamicMeshElement_0 Execution%20Dependency)
  (connect Branch_1 Output%20B StaticMeshToDynamicMeshElement_1 Execution%20Dependency)
  (connect MeshSampler_1 Out TransformPoints_1 In)
  (connect AddAttribute_0 Out TransformPoints_3 In))
