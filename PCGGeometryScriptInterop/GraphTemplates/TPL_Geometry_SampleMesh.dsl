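; PCG graph template that samples a static mesh and spawns debug axis tripods on the sampled points, alongside the mesh itself.
; Mesh Sampler takes one point per vertex, which is scaled, then copied onto a point created in original component space.
; The dot of each point's up vector with a normalized DotVector drives a range filter before tripod spawning.
; The created point also carries the Mesh attribute for a by-attribute spawner.
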(graph TPL_Geometry_SampleMesh
  (params (Mesh :type (SoftObject /Script/Engine.StaticMesh) :default /Engine/BasicShapes/Sphere.Sphere) (Scale :type Double :default 0.5) (DotVector :type Vector :default (0.0 0.0 1.0)) (FilterMin :type Float :default -1.0) (FilterMax :type Float :default 1.0))
  (bind TransformPoints_0 (Transform%20Points :ScaleMax (0.5 0.5 0.5) :ScaleMin (0.5 0.5 0.5) :ScaleMax-param Scale :ScaleMin-param Scale))
  (bind CreatePoints_0 (Create%20Points :CoordinateSpace OriginalComponent))
  (bind CopyPoints_0 (Copy%20Points))
  ;[0 0 0 0.5] Sample Mesh
  (bind MeshSampler_0 (Mesh%20Sampler :bExtractMeshFromInput true :bHasDefaultColorOverride false :SamplingMethod OnePointPerVertex :In Mesh))
  ;[0 0 0 0.5] Filter Points based on the dot between the specified Dot Vector and the points up vector
  (bind AttributeVectorOp_1 (Attribute%20Vector%20Op :Operation Normalize :In DotVector))
  (bind AttributeVectorOp_0 (Attribute%20Vector%20Op :InputSource1 $Rotation.Up :Operation Dot :OutputTarget Dot))
  (bind AttributeFilterRange_2 (Filter%20Attribute%20Elements%20by%20Range :MaxThreshold (struct :ThresholdAttribute @Last :bUseSpatialQuery false) :MinThreshold (struct :ThresholdAttribute @Last :bUseSpatialQuery false) :TargetAttribute @Last :FilterMax FilterMax :FilterMin FilterMin))
  ;[0 0 0 0.5] Spawn Mesh on Sampled Points
  (bind StaticMeshSpawner_0 (Static%20Mesh%20Spawner :MeshSelectorParameters (class:/Script/PCG.PCGMeshSelectorWeighted :MeshEntries [(struct :Descriptor (struct :StaticMesh /PCG/DebugObjects/PCG_AxisTripod.PCG_AxisTripod :bCanEverAffectNavigation true))])))
  ;[0 0 0 0.5] Spawn Sampled Mesh by Attribute
  (bind AddAttribute_0 (Add%20Attribute :InputSource Mesh :OutputTarget Mesh :Attributes Mesh))
  (bind StaticMeshSpawner_1 (Static%20Mesh%20Spawner :MeshSelectorParameters (class:/Script/PCG.PCGMeshSelectorByAttribute :AttributeName MEsh) :MeshSelectorType /Script/PCG.PCGMeshSelectorByAttribute))
  (connect CreatePoints_0 Out AddAttribute_0 In)
  (connect AttributeVectorOp_0 Out AttributeFilterRange_2 In)
  (connect CopyPoints_0 Out AttributeVectorOp_0 InA)
  (connect AttributeVectorOp_1 Out AttributeVectorOp_0 InB)
  (connect TransformPoints_0 Out CopyPoints_0 Source)
  (connect CreatePoints_0 Out CopyPoints_0 Target)
  (connect AttributeFilterRange_2 InsideFilter StaticMeshSpawner_0 In)
  (connect AddAttribute_0 Out StaticMeshSpawner_1 In)
  (connect MeshSampler_0 Out TransformPoints_0 In))
